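# Niagara effect for the Krampus glider, apparently a converted test version, adding colored smoke and twin trails.
# MedSmoke and MedSmokeLocal spawn offset flipbook sprites that drift along +X, shift from red to blue over life, and size by User.Global Scale.
# Trail_L and Trail_R emit short-lived white ribbons from skeletal mesh sockets FX_Trail_L2/R2, with random upward acceleration.
# The socket-based MedSmoke_Trailing emitter and the Empty emitter are disabled.
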NiagaraSystem /BRCosmetics/Gadgets/Assets/VinderTech_GliderChute/Glider_Krampus/FX/NS_KrampusGlider_Swirl_Test_Converted
dependencies {
 usedDataInterfaces [&NiagaraDataInterfaceSkeletalMesh, &NiagaraDataInterfaceCurve, &NiagaraDataInterfaceVectorCurve]
 usedDynamicInputs []
 usedModules [&SystemState]
}
emitters {
 Empty {
  bEnabled=false
  emitterUpdateScript [
   &EmitterState()
   &SpawnRate(SpawnRate=30, Use Spawn Probability=false)
  ]
  particleSpawnScript [&InitializeParticle(Color Mode=::NewEnumerator0, Lifetime=5, UsePositionOffset=false, Write Material Random=true)]
  particleUpdateScript [&ParticleState()]
  rendererClasses [&NiagaraSpriteRendererProperties]
  renderers [{rendererClass=&NiagaraSpriteRendererProperties}]
 }
 MedSmoke {
  emitterUpdateScript [
   &EmitterState()
   &SpawnRate(SpawnRate=8, Use Spawn Probability=false)
  ]
  particleSpawnScript [
   &InitializeParticle(
    Lifetime=1.4
    Sprite Rotation Angle=&RandomRangeFloat(Minimum=0, Maximum=1)
    Sprite Rotation Mode=::NewEnumerator1
    Sprite Size=&Vector2DFromFloat(Value=&Multiply_Float(A=&RandomRangeFloat(Minimum=110, Maximum=160), B=$User.Global Scale))
    Sprite Size Mode=::NewEnumerator4
    UsePositionOffset=false
    Write Material Random=true
   )
   &Color(Color=&MakeLinearColorFromVectorAndFloat(
     "Vector (RGB)"=<0.7 0.7 0.7>
     "Float (Alpha)"=&ClampFloat(Float=&RandomRangeFloat(Minimum=0.5, Maximum=0.9))
    ))
   &CameraOffset(Camera Offset Amount=&RandomRangeFloat(Minimum=10, Maximum=40))
   &DynamicMaterialParameters(
    Index 0 Param 1=&RandomRangeFloat(Minimum=0, Maximum=1)
    Index 0 Param 2=&RandomRangeFloat(Minimum=1, Maximum=2)
    Index 0 Param 3=25
    Module.Param0WriteEnabled=true
    Module.Param1WriteEnabled=true
    Module.Param2WriteEnabled=true
    Module.Param3WriteEnabled=false
   )
   &ShapeLocation(
    MaskSpawnGroup=false
    Offset=&RandomRangeVector(Minimum=<-40 -6 -4>, Maximum=<-60 6 12>)
    Offset Mode=::NewEnumerator0
    Override Seed=false
    Sphere Radius=&RandomRangeFloat(Minimum=8, Maximum=16)
    UseOwnerScale=true
   )
   &AddVelocity(Velocity=<150 0 0>)
  ]
  particleUpdateScript [
   &ParticleState()
   &ScaleSpriteSize(Uniform Curve Sprite Scale=DataInterface() {propertyValues {Curve {keys [{tangentMode=RCTM_User, time=0, value=0.2}, {tangentMode=RCTM_User, time=0.9, value=1}]}}})
   &ScaleColor(
    Scale Alpha=&FloatFromCurve(FloatCurve=DataInterface() {propertyValues {Curve {keys [
         <at 0 0 RCIM_Cubic>
         {arriveTangent=1.1078, interpMode=RCIM_Cubic, leaveTangent=1.1078, time=0.324, value=0.9889}
         {arriveTangent=-1.4628, interpMode=RCIM_Cubic, leaveTangent=-1.4628, time=0.6536, value=0.7241}
         <at 1 0 RCIM_Cubic>
        ]}}})
    Scale RGB=&VectorFromCurve(VectorCurve=DataInterface() {propertyValues {
       XCurve {keys [<at 0 1>,<at 0.8 0.0423>,<at 1 0.2997>]}
       YCurve {keys [<at 0 0>,<at 0.8 0.3498>,<at 1 0.569>]}
       ZCurve {keys [<at 0 0.0556>,<at 0.8 1>,<at 1 0.9062>]}
      }})
    ScaleA=true
    ScaleRGB=true
   )
   &DynamicMaterialParameters(
    Index 0 Param 4=0.5
    Module.Param0WriteEnabled=false
    Module.Param1WriteEnabled=false
    Module.Param2WriteEnabled=false
    Module.Param3WriteEnabled=true
   ) {module=DynamicMaterialParameters001}
   &SubUVAnimation(Start Frame=24)
   &Drag(Drag=0.8, Use Linear Drag=true, Use Rotational Drag=true)
   &SolveForcesAndVelocity(Clamp Velocity=false, Limit Acceleration=false)
  ]
  rendererClasses [&NiagaraSpriteRendererProperties]
  renderers [{rendererClass=&NiagaraSpriteRendererProperties}]
 }
 MedSmokeLocal {
  emitterUpdateScript [
   &EmitterState()
   &SpawnRate(SpawnRate=6, Use Spawn Probability=false)
  ]
  particleSpawnScript [
   /SetVariables() {isSetParameters=true}
   &InitializeParticle(
    Lifetime=&RandomRangeFloat(Minimum=0.8, Maximum=1)
    Sprite Rotation Angle=&RandomRangeFloat(Minimum=0, Maximum=1)
    Sprite Rotation Mode=::NewEnumerator1
    Sprite Size=&Multiply_Vector2D_ByFloat(Float=$User.Global Scale, Vector2D=<100 100>)
    Sprite Size Mode=::NewEnumerator4
    UsePositionOffset=false
    Write Material Random=true
   )
   &Color(Color=&MakeLinearColorFromVectorAndFloat(
     "Vector (RGB)"=<0.7 0.7 0.7>
     "Float (Alpha)"=&ClampFloat(Float=&RandomRangeFloat(Minimum=0.5, Maximum=0.7))
    ))
   &CameraOffset(Camera Offset Amount=&RandomRangeFloat(Minimum=25, Maximum=30))
   &DynamicMaterialParameters(
    Index 0 Param 1=&RandomRangeFloat(Minimum=0, Maximum=1)
    Index 0 Param 2=&RandomRangeFloat(Minimum=1, Maximum=2)
    Module.Param0WriteEnabled=true
    Module.Param1WriteEnabled=true
    Module.Param2WriteEnabled=false
    Module.Param3WriteEnabled=false
   )
   &ShapeLocation(
    MaskSpawnGroup=false
    Offset=&RandomRangeVector(Minimum=<-70 0 -32>, Maximum=<-100 0 40>)
    Offset Mode=::NewEnumerator0
    Override Seed=false
    Sphere Radius=&RandomRangeFloat(Minimum=2, Maximum=5)
    UseOwnerScale=true
   )
   &AddVelocity(Velocity=&RandomRangeVector(Minimum=<350 -150 0>, Maximum=<450 150 0>))
  ]
  particleUpdateScript [
   &ParticleState()
   &ScaleSpriteSize()
   &ScaleColor(
    Scale Alpha=&FloatFromCurve(FloatCurve=DataInterface() {propertyValues {Curve {keys [
         <at 0 0 RCIM_Cubic>
         {arriveTangent=1.3159, interpMode=RCIM_Cubic, leaveTangent=1.3159, time=0.1075, value=1.2249}
         {arriveTangent=-1.3725, interpMode=RCIM_Cubic, leaveTangent=-1.3725, time=0.6536, value=0.8601}
         <at 1 0 RCIM_Cubic>
        ]}}})
    Scale RGB=&VectorFromCurve(VectorCurve=DataInterface() {propertyValues {
       XCurve {keys [<at 0 1>,<at 0.8 0.0423>,<at 1 0.2997>]}
       YCurve {keys [<at 0 0>,<at 0.8 0.3498>,<at 1 0.569>]}
       ZCurve {keys [<at 0 0.0556>,<at 0.8 1>,<at 1 0.9062>]}
      }})
    ScaleA=true
    ScaleRGB=true
   )
   &SubUVAnimation(Start Frame=24) {module=SubUVAnimation001}
   &DynamicMaterialParameters(
    Index 0 Param 3=32
    Index 0 Param 4=0.5
    Module.Param0WriteEnabled=false
    Module.Param1WriteEnabled=false
    Module.Param2WriteEnabled=true
    Module.Param3WriteEnabled=true
   ) {module=DynamicMaterialParameters001}
   &Drag(Drag=1.4, Use Linear Drag=true, Use Rotational Drag=true)
   &AccelerationForce(Acceleration=$Particles.Acceleration)
   &SolveForcesAndVelocity(Clamp Velocity=false, Limit Acceleration=false)
  ]
  rendererClasses [&NiagaraSpriteRendererProperties]
  renderers [{rendererClass=&NiagaraSpriteRendererProperties}]
 }
 MedSmoke_Trailing {
  bEnabled=false
  emitterUpdateScript [
   &EmitterState()
   &SpawnRate(SpawnRate=10, Use Spawn Probability=false)
  ]
  particleSpawnScript [
   /SetVariables() {isSetParameters=true}
   &InitializeParticle(
    Lifetime=&RandomRangeFloat(Minimum=0.3, Maximum=0.4)
    Sprite Rotation Angle=&RandomRangeFloat(Minimum=0, Maximum=1)
    Sprite Rotation Mode=::NewEnumerator1
    Sprite Size=&Multiply_Vector2D_ByFloat(Float=$User.Global Scale, Vector2D=<100 100>)
    Sprite Size Mode=::NewEnumerator4
    UsePositionOffset=false
    Write Material Random=true
   )
   &Color(Color=&MakeLinearColorFromVectorAndFloat(
     "Vector (RGB)"=<0.7 0.7 0.7>
     "Float (Alpha)"=&ClampFloat(Float=&RandomRangeFloat(Minimum=0.3, Maximum=0.5))
    ))
   &SkeletalMeshLocation(
    MaskSpawnGroup=false
    Mesh Sampling Type=::NewEnumerator5
    OverrideRandomSeed=false
    Skeletal Mesh=DataInterface() {propertyValues {
      DefaultMesh=null
      ExcludeBoneName=root
      OptionalComponentClass=null
      PreviewMesh=/Game/Gadgets/Assets/VinderTech_GliderChute/Glider_Krampus/Meshes/Glider_Krampus_backCLoth_MESH
      SkinningMode=SkinOnTheFly
      SoftSourceActor=null
      SourceMode=Default
      UvSetIndex=0
      WholeMeshLOD=-1
      bExcludeBone=false
      bReadDeformedGeometry=true
      bRequireCurrentFrameData=true
      ComponentTags []
      FilteredBones []
      FilteredSockets [FX_Trail_R2,FX_Trail_L2]
      MeshUserParameter {parameter {name=None, typeDefHandle {registeredTypeIndex=1}, varData []}}
      SamplingRegions []
     }}
   )
   &CameraOffset(Camera Offset Amount=&RandomRangeFloat(Minimum=25, Maximum=30))
   &DynamicMaterialParameters(
    Index 0 Param 1=&RandomRangeFloat(Minimum=0, Maximum=1)
    Index 0 Param 2=&RandomRangeFloat(Minimum=1, Maximum=2)
    Index 0 Param 3=&RandomRangeFloat(Minimum=25, Maximum=30)
    Module.Param0WriteEnabled=true
    Module.Param1WriteEnabled=true
    Module.Param2WriteEnabled=true
    Module.Param3WriteEnabled=false
   )
   &ShapeLocation(
    MaskSpawnGroup=false
    Offset=&RandomRangeVector(Minimum=<-70 0 -15>, Maximum=<-100 0 40>)
    Offset Mode=::NewEnumerator0
    Override Seed=false
    Sphere Radius=&RandomRangeFloat(Minimum=10, Maximum=25)
    UseOwnerScale=true
   )
   &InheritVelocity(Inherited Velocity Speed Limit=2000, Speed Limit=true, UseSpeedThreshold=false)
   &AddVelocity(Velocity=&RandomRangeVector(Minimum=<350 -150 0>, Maximum=<450 150 0>))
   &ApplyInitialForces()
  ]
  particleUpdateScript [
   &ParticleState()
   &ScaleSpriteSize()
   &ScaleColor(
    Scale Alpha=&FloatFromCurve(FloatCurve=DataInterface() {propertyValues {Curve {keys [<at 0 0 RCIM_Cubic>,<at 0.4897 0.7749 RCIM_Cubic>,<at 1 0 RCIM_Cubic>]}}})
    Scale RGB=&VectorFromCurve(VectorCurve=DataInterface() {propertyValues {
       XCurve {keys [<at 0 1>,<at 0.8 0.0423>,<at 1 0.2997>]}
       YCurve {keys [<at 0 0>,<at 0.8 0.3498>,<at 1 0.569>]}
       ZCurve {keys [<at 0 0.0556>,<at 0.8 1>,<at 1 0.9062>]}
      }})
    ScaleA=true
    ScaleRGB=true
   )
   &SubUVAnimation(Start Frame=24)
   &DynamicMaterialParameters(
    Index 0 Param 4=0.5
    Module.Param0WriteEnabled=false
    Module.Param1WriteEnabled=false
    Module.Param2WriteEnabled=false
    Module.Param3WriteEnabled=true
   ) {module=DynamicMaterialParameters001}
   &Drag(Drag=8, Use Linear Drag=true, Use Rotational Drag=true)
   &AccelerationForce(Acceleration=$Particles.Acceleration)
   &SolveForcesAndVelocity(Clamp Velocity=false, Limit Acceleration=false)
  ]
  rendererClasses [&NiagaraSpriteRendererProperties]
  renderers [{rendererClass=&NiagaraSpriteRendererProperties}]
 }
 Trail_L {
  emitterUpdateScript [
   &EmitterState()
   &SpawnRate(SpawnRate=80, Use Spawn Probability=false)
  ]
  particleSpawnScript [
   /SetVariables(Particles.Acceleration:Vector3f=&RandomRangeVector(Minimum=<0 0 100>, Maximum=<0 0 250>)) {isSetParameters=true}
   &InitializeRibbon(
    Lifetime=0.2
    Position:NiagaraPosition=&ConvertVectorToPosition(Input Position=<-10 0 0>, Conversion Operation:ENiagara_LWCConvertVecToPos=::NewEnumerator0)
    Position Mode:ENiagara_PositionInitializationMode=::NewEnumerator1
    Ribbon Width=50
    UsePositionOffset=false
    Write Material Random=true
   )
   &Color(Color=&MakeLinearColorFromVectorAndFloat("Vector (RGB)"=<1.5 1.5 1.5>, "Float (Alpha)"=&ClampFloat(Float=0.5)))
   &AddVelocity(Velocity=<450 0 0>)
   &SkeletalMeshLocation(
    MaskSpawnGroup=false
    Mesh Sampling Type=::NewEnumerator5
    OverrideRandomSeed=false
    Skeletal Mesh=DataInterface() {propertyValues {
      DefaultMesh=null
      ExcludeBoneName=root
      OptionalComponentClass=null
      PreviewMesh=&Glider_Krampus_backCLoth_MESH
      SkinningMode=SkinOnTheFly
      SoftSourceActor=null
      SourceMode=Default
      UvSetIndex=0
      WholeMeshLOD=-1
      bExcludeBone=false
      bReadDeformedGeometry=true
      bRequireCurrentFrameData=true
      ComponentTags []
      FilteredBones []
      FilteredSockets [FX_Trail_L2]
      MeshUserParameter {parameter {name=None, typeDefHandle {registeredTypeIndex=1}, varData []}}
      SamplingRegions []
     }}
   )
  ]
  particleUpdateScript [
   /SetVariables(Particles.ScaledRibbonWidth:NiagaraFloat=&Multiply_Float(
     A=$Particles.RibbonWidth
     B=&FloatFromCurve(
      FloatCurve=DataInterface() {propertyValues {Curve {keys [{arriveTangent=0.1239, interpMode=RCIM_Cubic, time=0, value=0.9627}, {interpMode=RCIM_Cubic, leaveTangent=-0.1002, time=0.9983, value=0.857}]}}}
      Scale Curve=0.5
     )
    )) {isSetParameters=true}
   &ParticleState()
   &AccelerationForce(Acceleration=$Particles.Acceleration)
   &SolveForcesAndVelocity(Clamp Velocity=false, Limit Acceleration=false)
   &ScaleColor(
    Scale Alpha=&FloatFromCurve(FloatCurve=DataInterface() {propertyValues {Curve {keys [<at 0 0 RCIM_Cubic>,<at 0.1991 0.995 RCIM_Cubic>,<at 1 0>]}}})
    ScaleA=true
    ScaleRGB=true
   )
   &DynamicMaterialParameters(
    Index 0 Param 1=1
    Index 0 Param 2=1
    Index 0 Param 3=0.1
    Index 0 Param 4=1
    Module.Param0WriteEnabled=true
    Module.Param1WriteEnabled=true
    Module.Param2WriteEnabled=true
    Module.Param3WriteEnabled=true
   )
  ]
  rendererClasses [&NiagaraRibbonRendererProperties]
  renderers [{rendererClass=&NiagaraRibbonRendererProperties}]
 }
 Trail_R {
  emitterUpdateScript [
   &EmitterState()
   &SpawnRate(SpawnRate=80, Use Spawn Probability=false)
  ]
  particleSpawnScript [
   /SetVariables(Particles.Acceleration:Vector3f=&RandomRangeVector(Minimum=<0 0 100>, Maximum=<0 0 250>)) {isSetParameters=true}
   &InitializeRibbon(
    Lifetime=0.2
    Position:NiagaraPosition=&ConvertVectorToPosition(Input Position=<-10 0 0>, Conversion Operation:ENiagara_LWCConvertVecToPos=::NewEnumerator0)
    Position Mode:ENiagara_PositionInitializationMode=::NewEnumerator1
    Ribbon Width=50
    UsePositionOffset=false
    Write Material Random=true
   )
   &Color(Color=&MakeLinearColorFromVectorAndFloat("Vector (RGB)"=<1.5 1.5 1.5>, "Float (Alpha)"=&ClampFloat(Float=0.5)))
   &AddVelocity(Velocity=<450 0 0>)
   &SkeletalMeshLocation(
    MaskSpawnGroup=false
    Mesh Sampling Type=::NewEnumerator5
    OverrideRandomSeed=false
    Skeletal Mesh=DataInterface() {propertyValues {
      DefaultMesh=null
      ExcludeBoneName=root
      OptionalComponentClass=null
      PreviewMesh=&Glider_Krampus_backCLoth_MESH
      SkinningMode=SkinOnTheFly
      SoftSourceActor=null
      SourceMode=Default
      UvSetIndex=0
      WholeMeshLOD=-1
      bExcludeBone=false
      bReadDeformedGeometry=true
      bRequireCurrentFrameData=true
      ComponentTags []
      FilteredBones []
      FilteredSockets [FX_Trail_R2]
      MeshUserParameter {parameter {name=None, typeDefHandle {registeredTypeIndex=1}, varData []}}
      SamplingRegions []
     }}
   )
  ]
  particleUpdateScript [
   /SetVariables(Particles.ScaledRibbonWidth:NiagaraFloat=&Multiply_Float(
     A=$Particles.RibbonWidth
     B=&FloatFromCurve(
      FloatCurve=DataInterface() {propertyValues {Curve {keys [{arriveTangent=0.1239, interpMode=RCIM_Cubic, time=0, value=0.9627}, {interpMode=RCIM_Cubic, leaveTangent=-0.1002, time=0.9983, value=0.857}]}}}
      Scale Curve=0.5
     )
    )) {isSetParameters=true}
   &ParticleState()
   &AccelerationForce(Acceleration=$Particles.Acceleration)
   &SolveForcesAndVelocity(Clamp Velocity=false, Limit Acceleration=false)
   &ScaleColor(
    Scale Alpha=&FloatFromCurve(FloatCurve=DataInterface() {propertyValues {Curve {keys [<at 0 0 RCIM_Cubic>,<at 0.1991 0.995 RCIM_Cubic>,<at 1 0>]}}})
    ScaleA=true
    ScaleRGB=true
   )
   &DynamicMaterialParameters(
    Index 0 Param 3=0.1
    Index 0 Param 4=1
    Module.Param0WriteEnabled=true
    Module.Param1WriteEnabled=true
    Module.Param2WriteEnabled=true
    Module.Param3WriteEnabled=true
   )
  ]
  rendererClasses [&NiagaraRibbonRendererProperties]
  renderers [{rendererClass=&NiagaraRibbonRendererProperties}]
 }
}
summary {
 systemName=NS_KrampusGlider_Swirl_Test_Converted
 emitters [
  {emitterName=Trail_L, rendererClasses [&NiagaraRibbonRendererProperties]}
  {emitterName=Trail_R, rendererClasses [&NiagaraRibbonRendererProperties]}
  {emitterName=MedSmokeLocal, rendererClasses [&NiagaraSpriteRendererProperties]}
  {emitterName=MedSmoke, rendererClasses [&NiagaraSpriteRendererProperties]}
  {bEnabled=false, emitterName=MedSmoke_Trailing, rendererClasses [&NiagaraSpriteRendererProperties]}
  {bEnabled=false, emitterName=Empty, rendererClasses [&NiagaraSpriteRendererProperties]}
 ]
 userVariables [User.Global Scale {type=NiagaraFloat, defaultValue=1.5, description=""}]
}
